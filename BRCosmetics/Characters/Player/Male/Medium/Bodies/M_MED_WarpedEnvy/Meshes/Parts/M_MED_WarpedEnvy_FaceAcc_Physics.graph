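# Physics asset for the face accessory part of the male medium WarpedEnvy character body, providing collision and simulation shapes for dynamic hair bones.
# Capsule and sphere bodies on the head, neck, upper spine and both clavicles approximate the upper body around the hair.
# Capsules cover a left front hair chain, its master bone, and one bone in each left and right hair strand.
# Only the front chain is constrained, with twist locked and swing limits loosening toward the tip.
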PhysicsAsset /BRCosmetics/Characters/Player/Male/Medium/Bodies/M_MED_WarpedEnvy/Meshes/Parts/M_MED_WarpedEnvy_FaceAcc_Physics
bodies {
 clavicle_l [{shapeType=Capsule, length=20.2598, radius=9.1372, shapeName=clavicle_l_capsule, center=<6.1014 1.3268 0.5446>, rotation=<rot 90 8.7581 0>}]
 clavicle_r [{
   shapeType=Capsule
   length=25.1186
   radius=8.9379
   shapeName=clavicle_l_capsule
   center=<-5.6809 -1.2447 -1.2994>
   rotation=<rot 84.1081 -168.7322 0.0004>
  }]
 dyn_hairMeshFront_l_01 [{shapeType=Capsule, length=1.8937, radius=0.6682, shapeName=dyn_hairMeshFront_l_01_capsule, center=<1.1843 0 0>, rotation=<rot 90 0 0>}]
 dyn_hairMeshFront_l_02 [{
   shapeType=Capsule
   length=5.2324
   radius=1.1682
   shapeName=dyn_hairMeshFront_l_02_capsule
   center=<3.0831 0 -0.1484>
   rotation=<rot 83.4024 -0.0003 -0.0004>
  }]
 dyn_hairMeshFront_l_03 [{shapeType=Capsule, length=7.3004, radius=1.1682, shapeName=dyn_hairMeshFront_l_03_capsule, center=<3.527 0 0>, rotation=<rot 90 0 0>}]
 dyn_hairMeshFront_l_04 [{shapeType=Capsule, length=6.2527, radius=0.9182, shapeName=dyn_hairMeshFront_l_04_capsule, center=<2.3769 0 0>, rotation=<rot 84.6719 0 0>}]
 dyn_hairMeshFront_l_MSTR [{shapeType=Capsule, length=0.5, radius=0.5, shapeName=dyn_hairMeshFront_l_MSTR_capsule, rotation=<rot 90 0 0>}]
 dyn_hairMesh_l_02 [{shapeType=Capsule, length=5.4966, radius=1.4572, shapeName=dyn_hairMesh_l_02_capsule, center=<2.9869 0 0>, rotation=<rot 90 0 0>}]
 dyn_hairMesh_r_02 [{
   shapeType=Capsule
   length=7.856
   radius=0.8341
   shapeName=dyn_hairMesh_r_02_capsule
   center=<3.1676 0 0.4>
   rotation=<rot 81.9034 179.9999 179.9999>
  }]
 head [
  {shapeType=Capsule, length=3.8424, radius=7.5071, center=<4.9817 -0.2 0>, rotation=<rot 90 0 0>}
  {shapeType=Capsule, length=7.4901, radius=6.8818, shapeName=head_capsule, center=<3.4471 3.7892 0>, rotation=<rot 90 0 16.2417>}
 ]
 neck_02 [
  {shapeType=Capsule, length=9.0767, radius=6.4432, shapeName=neck_02_capsule, center=<1.8504 1.2829 0>, rotation=<rot 90 0 0>}
  {shapeType=Capsule, length=0, radius=1.0423, shapeName=neck_02_capsule}
  {shapeType=Capsule, length=0, radius=2.0306, shapeName=neck_02_capsule, center=<-2.5793 0 0>}
 ]
 spine_05 [
  {shapeType=Sphere, radius=10.8031, shapeName=spine_05_sphere, center=<4.8672 -0.7071 0>}
  {shapeType=Capsule, length=15.805, radius=11.3472, shapeName=spine_05_capsule, center=<-0.4404 0 0>, rotation=<rot 0 -90 0>}
  {shapeType=Capsule, length=4.1356, radius=11.0034, shapeName=spine_05_capsule, center=<6.3334 3.2484 0>}
 ]
}
constraints [
 {bone1Name=dyn_hairMeshFront_l_02, bone2Name=dyn_hairMeshFront_l_01, swing1LimitDegrees=20, swing2LimitDegrees=10, twistLimitDegrees=0}
 {bone1Name=dyn_hairMeshFront_l_03, bone2Name=dyn_hairMeshFront_l_02, swing1LimitDegrees=25, swing2LimitDegrees=20, twistLimitDegrees=0}
 {bone1Name=dyn_hairMeshFront_l_04, bone2Name=dyn_hairMeshFront_l_03, swing1LimitDegrees=30, swing2LimitDegrees=20, twistLimitDegrees=0}
]
